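# Niagara effect for the CiderPetalClover backpack cosmetic, built from three mesh emitters: Base, Spline and Spline02.
# Each burst-spawns non-expiring particles with position and camera offsets and dynamic material parameters, drawn by three mesh renderers.
# User.ColorLerp blends Base from blue to gold across its main color and two extra mesh color attributes.
# The Spline emitters take Main, Energy and Glow colors from shared system variables.
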NiagaraSystem /BRCosmetics/Accessories/FORT_Backpacks/Backpack_M_MED_CiderPetalClover/FX/NS_Backpack_CiderPartalClover
dependencies {
 usedDataInterfaces []
 usedDynamicInputs []
 usedModules [
  &self:SystemSpawnScript.SystemScriptSource.SystemScriptGraph.NiagaraNodeAssignment_1.SetVariables_#1
  &self:SystemSpawnScript.SystemScriptSource.SystemScriptGraph.NiagaraNodeAssignment_2.SetVariables_#2
  &self:SystemSpawnScript.SystemScriptSource.SystemScriptGraph.NiagaraNodeAssignment_3.SetVariables_#3
  &SystemState
 ]
}
emitters {
 Base {
  emitterUpdateScript [
   &EmitterState(Life Cycle Mode=::NewEnumerator1, Loop Behavior=::NewEnumerator1, Loop Duration Mode=::NewEnumerator1, UseLoopDelay=false)
   &SpawnBurst_Instantaneous(Use Spawn Probability=false, UseLoopCountLimit=false)
  ]
  particleSpawnScript [
   &InitializeParticle(
    Color Mode=::NewEnumerator0
    Lifetime=5
    Mesh Scale Mode=::NewEnumerator3
    Position Offset=<-3.5 -3.8 0>
    UsePositionOffset=true
    Write Material Random=true
   )
   &DynamicMaterialParameters(
    Index 0 Param 1=0.2
    Index 1 Param 1=0.5
    Module.Index0Float0Write=true
    Module.Index0Float1Write=true
    Module.Index0Float2Write=true
    Module.Index0Float3Write=true
    Module.Index1Float0Write=true
    Module.Index1Float1Write=true
    Module.Index1Float2Write=true
    Module.Index1Float3Write=true
    Module.Index2Float0Write=true
    Module.Index2Float1Write=true
    Module.Index2Float2Write=true
    Module.Index2Float3Write=true
    Write Parameter Index 1=::NewEnumerator1
    Write Parameter Index 2=::NewEnumerator1
   )
   &CameraOffset(Camera Offset Amount=0.5)
  ]
  particleUpdateScript [
   &ParticleState(Kill Particles When Lifetime Has Elapsed=false, Let Infinitely Lived Particles Die When Emitter Deactivates=true)
   &Color(Color=&LerpLinearColors(StartColor=<rgba 0 4.5 30 1>, EndColor=<rgba 15 10 0 1>, LerpFactor=$User.ColorLerp))
   &FN_Mesh_Attributes(
    Color=&LerpLinearColors(StartColor=<rgba 0.2 0.8 2 1>, EndColor=<rgba 2 1.4 0.2 1>, LerpFactor=$User.ColorLerp)
    Create Color Attribute=true
    Module Output=::NewEnumerator1
   ) {module=FN_Mesh_Attributes001}
   &FN_Mesh_Attributes(
    Color=&Multiply_LinearColorByFloat(
     Linear Color=&LerpLinearColors(StartColor=<rgba 0.2 0.3067 1 1>, EndColor=<rgba 1 0.7333 0.2 1>, LerpFactor=$User.ColorLerp)
     Float=0.5
    )
    Create Color Attribute=true
    "Create Normalized Age "=true
    Module Output=::NewEnumerator1
   )
  ]
  rendererClasses [&NiagaraMeshRendererProperties]
  renderers [
   {rendererClass=&NiagaraMeshRendererProperties, rendererIndex=0}
   {rendererClass=&NiagaraMeshRendererProperties, rendererIndex=1}
   {rendererClass=&NiagaraMeshRendererProperties, rendererIndex=2}
  ]
 }
 Spline {
  emitterUpdateScript [
   &EmitterState(
    Inactive Response=::NewEnumerator1
    Life Cycle Mode=::NewEnumerator1
    Loop Behavior=::NewEnumerator1
    Loop Duration Mode=::NewEnumerator1
    UseLoopDelay=false
   )
   &SpawnBurst_Instantaneous(Use Spawn Probability=false, UseLoopCountLimit=false)
  ]
  particleSpawnScript [
   &InitializeParticle(Lifetime=5, Mesh Scale Mode=::NewEnumerator3, Position Offset=<-2.5 -3 0>, UsePositionOffset=true, Write Material Random=true)
   &DynamicMaterialParameters(
    Index 0 Param 1=1
    Index 1 Param 1=1.5
    Index 2 Param 1=0.5
    Module.Index0Float0Write=true
    Module.Index0Float1Write=true
    Module.Index0Float2Write=true
    Module.Index0Float3Write=true
    Module.Index1Float0Write=true
    Module.Index1Float1Write=true
    Module.Index1Float2Write=true
    Module.Index1Float3Write=true
    Module.Index2Float0Write=true
    Module.Index2Float1Write=true
    Module.Index2Float2Write=true
    Module.Index2Float3Write=true
    Write Parameter Index 1=::NewEnumerator1
    Write Parameter Index 2=::NewEnumerator1
   )
   &CameraOffset()
  ]
  particleUpdateScript [
   &ParticleState(Kill Particles When Lifetime Has Elapsed=false, Let Infinitely Lived Particles Die When Emitter Deactivates=true)
   &Color(Color=$System.Spline Main Color)
   &FN_Mesh_Attributes(Color=$System.Spline Energy Color, Create Color Attribute=true, Module Output=::NewEnumerator1) {module=FN_Mesh_Attributes001}
   &FN_Mesh_Attributes(Color=$System.Spline Glow Color, Create Color Attribute=true, Module Output=::NewEnumerator1)
  ]
  rendererClasses [&NiagaraMeshRendererProperties]
  renderers [
   {rendererClass=&NiagaraMeshRendererProperties, rendererIndex=0}
   {rendererClass=&NiagaraMeshRendererProperties, rendererIndex=1}
   {rendererClass=&NiagaraMeshRendererProperties, rendererIndex=2}
  ]
 }
 Spline02 {
  emitterUpdateScript [
   &EmitterState(
    Inactive Response=::NewEnumerator1
    Life Cycle Mode=::NewEnumerator1
    Loop Behavior=::NewEnumerator1
    Loop Duration Mode=::NewEnumerator1
    UseLoopDelay=false
   )
   &SpawnBurst_Instantaneous(Use Spawn Probability=false, UseLoopCountLimit=false)
  ]
  particleSpawnScript [
   &InitializeParticle(
    Color Mode=::NewEnumerator0
    Lifetime=5
    Mesh Scale Mode=::NewEnumerator3
    Position Offset=<-3.2 -4 0>
    UsePositionOffset=true
    Write Material Random=true
   )
   &DynamicMaterialParameters(
    Index 0 Param 1=1
    Index 1 Param 1=1.5
    Index 2 Param 1=0.5
    Module.Index0Float0Write=true
    Module.Index0Float1Write=true
    Module.Index0Float2Write=true
    Module.Index0Float3Write=true
    Module.Index1Float0Write=true
    Module.Index1Float1Write=true
    Module.Index1Float2Write=true
    Module.Index1Float3Write=true
    Module.Index2Float0Write=true
    Module.Index2Float1Write=true
    Module.Index2Float2Write=true
    Module.Index2Float3Write=true
    Write Parameter Index 1=::NewEnumerator1
    Write Parameter Index 2=::NewEnumerator1
   )
   &CameraOffset(Camera Offset Amount=20)
  ]
  particleUpdateScript [
   &ParticleState(Kill Particles When Lifetime Has Elapsed=false, Let Infinitely Lived Particles Die When Emitter Deactivates=true)
   &Color(Color=$System.Spline Main Color)
   &FN_Mesh_Attributes(Color=$System.Spline Energy Color, Create Color Attribute=true, Module Output=::NewEnumerator1) {module=FN_Mesh_Attributes001}
   &FN_Mesh_Attributes(Color=$System.Spline Glow Color, Create Color Attribute=true, Module Output=::NewEnumerator1)
  ]
  rendererClasses [&NiagaraMeshRendererProperties]
  renderers [
   {rendererClass=&NiagaraMeshRendererProperties, rendererIndex=0}
   {rendererClass=&NiagaraMeshRendererProperties, rendererIndex=1}
   {rendererClass=&NiagaraMeshRendererProperties, rendererIndex=2}
  ]
 }
}
summary {
 systemName=NS_Backpack_CiderPartalClover
 emitters [
  {emitterName=Base, rendererClasses [&NiagaraMeshRendererProperties]}
  {emitterName=Spline, rendererClasses [&NiagaraMeshRendererProperties]}
  {emitterName=Spline02, rendererClasses [&NiagaraMeshRendererProperties]}
 ]
 userVariables [User.ColorLerp {type=NiagaraFloat, defaultValue=0, description=""}]
}
