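# Anim Blueprint for the CiderPetalClover backpack cosmetic, which repositions its pack bone on top of the input pose.
# Two chained additive bone-space Transform Bone nodes apply fixed offsets; the second adds roll behind an unwired bool alpha with blending.
# On customization loaded, the owning component is hidden.
# A spine-delta-to-roll calculation for backpackRotate is not linked to the update event, and the AnimGraph never reads that variable.
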AnimBlueprint /BRCosmetics/Accessories/FORT_Backpacks/Backpack_M_MED_CiderPetalClover/Meshes/M_MED_CiderPetalClover_Pack_AnimBP
target_skeleton=&/M_MED_CiderPetalClover_Pack_Skeleton
graphs [
 AnimGraph {nodes [
   n1 = "Transform (Modify) Bone - Bone: pack"(ComponentPose=Local To Component(LocalPose=Input Pose()), Translation="0,-2,-3", Rotation="0,0,0") {
    rotationMode=BMM_Additive
    rotationSpace=BCS_BoneSpace
    translationMode=BMM_Additive
    alphaBoolBlend {blendOption=Linear}
    boneToModify {boneName=pack}
    translation {y=-2, z=-3}
   }
   n2 = "Transform (Modify) Bone - Bone: pack"!(ComponentPose=@n1, bAlphaBoolEnabled:bool, Translation="-3,-4,0", Rotation="0,0,15") -> (Pose) {
    alphaInputType=Bool
    rotationMode=BMM_Additive
    rotationSpace=BCS_BoneSpace
    translationMode=BMM_Additive
    alphaBoolBlend {blendInTime=0.1, blendOption=Linear, blendOutTime=0.1}
    boneToModify {boneName=pack}
    rotation {roll=15}
    translation=<-3 -4>
   }
   Output Pose(Result=Component To Local(ComponentPose=@n2)) {layerGroup=None}
  ]}
 EventGraph {nodes [
   Event OnCharacterCustomizationLoaded()
   > SetVisibility!(execute, self:Scene Component Object Reference=GetOwningComponent(), bNewVisibility:bool=false, bPropagateToChildren:bool=false) -> (then)
   Event BlueprintUpdateAnimation()
   n1 = GetOwnerSkeletalMeshComponent()
   n2 = BreakRotator(InRot=BreakTransform(InTransform=GetDeltaTransformFromRefPose(self=@n1, BoneName=spine_03, BaseName=spine_05)).Rotation)
   Set backpackRotate(backpackRotate:Rotator=MakeRotator(Roll=MapRangeClamped(Value=@n2.Yaw, InRangeA=0, InRangeB=15, OutRangeB=20))) -> (Output_Get:Rotator)
  ]}
]
variables [
 backpackRotate {type=Rotator}
 diving? {type=number}
]
